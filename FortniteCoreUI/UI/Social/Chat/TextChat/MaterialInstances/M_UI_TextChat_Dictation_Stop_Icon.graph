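# UI material for the stop icon in text chat dictation, drawn procedurally as a rounded square or rectangle.
# A fixed pixel inset is converted to normalized units and subtracted from one to size the shape.
# The shared CoreUI rounded-mask function turns that size and a pixel-based corner radius into the mask.
# The translucent UI-domain material also exposes an Icon Color parameter, white by default.
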Material /FortniteCoreUI/UI/Social/Chat/TextChat/MaterialInstances/M_UI_TextChat_Dictation_Stop_Icon
graph {
 properties {bAutomaticallySetUsageInEditor=false, blendMode=BLEND_Translucent, materialDomain=MD_UI, usedShadingModels=Default Lit}
 nodes [
  n1 = Constant() {r=8}
  MaterialFunctionCall(
   Size=Subtract(
    A=Constant() {r=1}
    B=MaterialFunctionCall!(Value in Pixels=@n1, Pixel Reference) {materialFunction=/FortUILibrary/MaterialFunctions/Basics/MF_UI_PixelsToNormalizedV4}."Normalized (V4)"
   )
   Corner Radius=Constant() {r=6}
   Radius in Pixels=Constant() {r=1}
  ) {materialFunction=&MF_UI_CoreUI_RoundedMask}
  VectorParameter() {parameterName=Icon Color, defaultValue=<rgba 1 1 1 1>}
 ]
}
